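# Physics asset for the DarkShaman pickaxe skeletal mesh, a two-handed melee cosmetic weapon.
# It holds a single capsule body on the Melee_TwoHanded_Handle bone, offset and rotated along the weapon as one rough collision volume.
# No constraints exist, so the asset defines no articulated or jointed simulation.
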PhysicsAsset /BRCosmetics/Weapons/FORT_Melee/Pickaxe_DarkShaman/Meshes/SK_Pickaxe_DarkShaman_PhysicsAsset
bodies {Melee_TwoHanded_Handle [{shapeType=Capsule, length=76.5922, radius=28.1295, center=<1.8407 1.0449 29.9348>, rotation=<rot -1.5999 -90 0.0741>}]}
constraints []
